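# Sound cue for the left-foot run footstep foley of the AbruptSudden biped cosmetic companion.
# A Random node picks among three equally weighted odd-numbered Walk waves, so the run reuses walk recordings with a slightly raised pitch and reduced volume.
# It uses the pets sound class, companion foley attenuation and a shared biped footstep concurrency set, at zero priority.
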SoundCue /CosmeticCompanions/Assets/Biped/AbruptSudden/Audio/Foley/SC_MImosa_Biped_AbruptSudden_Run_L
root=Random_0
nodes [
 {id=Random_0, children [WavePlayer_15, WavePlayer_13, WavePlayer_11], properties {weights [1,1,1]}}
 {
  id=WavePlayer_11
  properties {soundWaveAssetPtr=/CosmeticCompanions/Assets/Biped/AbruptSudden/Audio/Foley/Wavs/Mimosa_Biped_AbruptSudden_Walk_05}
 }
 {
  id=WavePlayer_13
  properties {soundWaveAssetPtr=/CosmeticCompanions/Assets/Biped/AbruptSudden/Audio/Foley/Wavs/Mimosa_Biped_AbruptSudden_Walk_03}
 }
 {
  id=WavePlayer_15
  properties {soundWaveAssetPtr=/CosmeticCompanions/Assets/Biped/AbruptSudden/Audio/Foley/Wavs/Mimosa_Biped_AbruptSudden_Walk_01}
 }
]
properties {
 attenuationSettings=/Game/Sounds/Attenuation/AS_Companion_Foley
 maxDistance=1200
 pitchMultiplier=1.1
 priority=0
 soundClassObject=&Fort_Creatures_Pets
 volumeMultiplier=0.65
 concurrencySet [/CosmeticCompanions/Assets/Audio_Shared/Foley/Biped/CONC_Mimosa_Biped_Footstep]
}
