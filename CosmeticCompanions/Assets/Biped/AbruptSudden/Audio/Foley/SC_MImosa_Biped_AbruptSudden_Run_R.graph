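# Sound cue for the run foley of the AbruptSudden biped cosmetic companion, apparently the right-foot step.
# A Random node picks with equal weight among three even-numbered Walk recordings, reused here with a raised pitch and reduced volume.
# It plays in the creature pets sound class with companion foley attenuation and a shared biped footstep concurrency set.
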SoundCue /CosmeticCompanions/Assets/Biped/AbruptSudden/Audio/Foley/SC_MImosa_Biped_AbruptSudden_Run_R
root=Random_0
nodes [
 {id=Random_0, children [WavePlayer_14, WavePlayer_12, WavePlayer_10], properties {weights [1,1,1]}}
 {
  id=WavePlayer_10
  properties {soundWaveAssetPtr=/CosmeticCompanions/Assets/Biped/AbruptSudden/Audio/Foley/Wavs/Mimosa_Biped_AbruptSudden_Walk_06}
 }
 {
  id=WavePlayer_12
  properties {soundWaveAssetPtr=/CosmeticCompanions/Assets/Biped/AbruptSudden/Audio/Foley/Wavs/Mimosa_Biped_AbruptSudden_Walk_04}
 }
 {
  id=WavePlayer_14
  properties {soundWaveAssetPtr=/CosmeticCompanions/Assets/Biped/AbruptSudden/Audio/Foley/Wavs/Mimosa_Biped_AbruptSudden_Walk_02}
 }
]
properties {
 attenuationSettings=/Game/Sounds/Attenuation/AS_Companion_Foley
 maxDistance=1200
 pitchMultiplier=1.1
 priority=0
 soundClassObject=&Fort_Creatures_Pets
 volumeMultiplier=0.6
 concurrencySet [/CosmeticCompanions/Assets/Audio_Shared/Foley/Biped/CONC_Mimosa_Biped_Footstep]
}
